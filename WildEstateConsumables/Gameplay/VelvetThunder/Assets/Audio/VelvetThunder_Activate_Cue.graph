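# Sound cue for activating the Velvet Thunder consumable, played as a positioned effect.
# A single wave player feeds the activate sound wave straight to the output, with no randomization or mixing.
# It uses the shared Fort item-interact attenuation with a limited audible range, routes through the loot reward sound class, and slightly boosts volume.
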SoundCue /WildEstateConsumables/Gameplay/VelvetThunder/Assets/Audio/VelvetThunder_Activate_Cue
root=WavePlayer_2
nodes [{id=WavePlayer_2, properties {soundWaveAssetPtr=&/VelvetThunder_Activate}}]
properties {
 attenuationSettings=/Game/Sounds/Attenuation/Fort_Item_Interact_Att
 maxDistance=1850
 soundClassObject=&Fort_Loot_Reward_PSM
 volumeMultiplier=1.1
}
