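# Ambient idle loop Sound Cue for the VelvetThunder consumable, classed as a looping loot ambient sound.
# A mixer layers a continuously looping idle bed with a slightly quieter, endlessly repeating rattle layer.
# The rattle layer waits a random delay, then plays one of six equally weighted variations with random downward pitch variation.
# It references shared payphone ambient attenuation and loot ambient concurrency assets.
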SoundCue /WildEstateConsumables/Gameplay/VelvetThunder/Assets/Audio/VelvetThunder_Ambient_Loop_Cue
root=Mixer_0
nodes [
 {id=Delay_1, children [Modulator_0], properties {delayMax=3.5, delayMin=1}}
 {id=Looping_0, children [Delay_1]}
 {id=Mixer_0, children [WavePlayer_1,Looping_0], properties {inputVolume [1,0.9]}}
 {id=Modulator_0, children [Random_0], properties {pitchMax=1, pitchMin=0.8}}
 {id=Random_0, children [WavePlayer_3,WavePlayer_4,WavePlayer_5,WavePlayer_6,WavePlayer_7,WavePlayer_8], properties {weights [1,1,1,1,1,1]}}
 {
  id=WavePlayer_1
  properties {bLooping=true, soundWaveAssetPtr=&/VelvetThunder_Idle_Loop}
 }
 {id=WavePlayer_3, properties {soundWaveAssetPtr=&/VelvetThunder_Idle_Rattles_01}}
 {id=WavePlayer_4, properties {soundWaveAssetPtr=&/VelvetThunder_Idle_Rattles_02}}
 {id=WavePlayer_5, properties {soundWaveAssetPtr=&/VelvetThunder_Idle_Rattles_03}}
 {id=WavePlayer_6, properties {soundWaveAssetPtr=&/VelvetThunder_Idle_Rattles_04}}
 {id=WavePlayer_7, properties {soundWaveAssetPtr=&/VelvetThunder_Idle_Rattles_05}}
 {id=WavePlayer_8, properties {soundWaveAssetPtr=&/VelvetThunder_Idle_Rattles_06}}
]
properties {
 attenuationSettings=/Game/Sounds/Attenuation/Fort_Payphone_Amb_Att
 maxDistance=1700
 soundClassObject=/Game/Sounds/FortSoundClassesAndModes/Fort_Loot_Ambient_Loop
 volumeMultiplier=0.8
 attenuationOverrides {
  bEnableFocusInterpolation=true
  bEnableListenerFocus=true
  distanceAlgorithm=Logarithmic
  falloffDistance=1000
  focusAttackInterpSpeed=20
  focusAzimuth=20
  focusReleaseInterpSpeed=10
  focusVolumeAttenuation=1.6
  lPFFrequencyAtMax=10000
  lPFRadiusMax=1000
  lPFRadiusMin=500
  nonFocusAzimuth=40
  nonFocusVolumeAttenuation=0.6
  nonSpatializedRadiusStart=200
  occlusionInterpolationTime=0.3
  occlusionLowPassFilterFrequency=3000
  occlusionVolumeAttenuation=0.75
  attenuationShapeExtents {x=0}
  customAttenuationCurve {editorCurveData {keys [<at 0 1 RCIM_Cubic>, {arriveTangent=-1, interpMode=RCIM_Cubic, leaveTangent=-1, time=0.5, value=0.1}, <at 1 0 RCIM_Cubic>]}}
 }
 concurrencyOverrides {bLimitToOwner=true, maxCount=1, resolutionRule=StopQuietest}
 concurrencySet [/Game/Sounds/Concurrency_Rules/Loot_Ambient_Conc]
}
